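# Material function that decodes the Water plugin's water info texture for fluid simulation materials.
# It samples the WaterVelocityAndHeight parameter at input UVs with a fixed mip level and shared wrap sampler.
# Custom HLSL nodes call WaterDataFunctions.ush decoders, using MaxFlowVelocity for flow and WaterZoneIndex for water and ground heights.
# It outputs velocity, both heights, depth as water minus terrain height, and raw RGBA.
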MaterialFunction /FortWater/FluidSimulation/Materials/Functions/Fluidsim_GetWaterData
graph {nodes [
  MaterialExpressionTextureSample_0 = TextureSample!(
   UVs=FunctionInput() {bUsePreviewValueAsDefault=true, inputName=UVs, inputType=FunctionInput_Vector2}
   Tex=TextureObjectParameter() {parameterName=WaterVelocityAndHeight, texture=&Black}
   Level
   Apply View MipBias
  ) {constMipValue=0, mipValueMode=TMVM_MipLevel, samplerSource=SSM_Wrap_WorldGroupSettings, texture=&Black}
  MaterialExpressionScalarParameter_5 = ScalarParameter() {defaultValue=1024, parameterName=MaxFlowVelocity}
  n1 = Custom!(WaterInfoSample=@MaterialExpressionTextureSample_0.RGBA, MaxVelocity=@MaterialExpressionScalarParameter_5) {
   code="return DecodeWaterInfoVelocity(WaterInfoSample, MaxVelocity);"
   description=Decode Flow From Water Info
   outputType=CMOT_Float2
   includeFilePaths [/Plugin/Water/Private/WaterDataFunctions.ush]
   inputs [
    {
     inputName=WaterInfoSample
     input {
      expression=&self:MaterialExpressionTextureSample_0
      mask=1
      maskA=1
      maskB=1
      maskG=1
      maskR=1
      outputIndex=5
     }
    }
    {
     inputName=MaxVelocity
     input {expression=&self:MaterialExpressionScalarParameter_5}
    }
   ]
  }
  FunctionOutput(None=@n1) {outputName=Velocity}
  MaterialExpressionScalarParameter_0 = ScalarParameter() {parameterName=WaterZoneIndex}
  n2 = Custom!(WaterInfoSample=@MaterialExpressionTextureSample_0.RGBA, WaterZoneIndex=@MaterialExpressionScalarParameter_0) {
   code="return DecodeWaterInfoZHeight(WaterInfoSample, WaterZoneIndex);"
   description=DecodeWaterInfoZHeight
   outputType=CMOT_Float1
   includeFilePaths [&ush]
   inputs [
    {
     inputName=WaterInfoSample
     input {
      expression=&self:MaterialExpressionTextureSample_0
      mask=1
      maskA=1
      maskB=1
      maskG=1
      maskR=1
      outputIndex=5
     }
    }
    {
     inputName=WaterZoneIndex
     input {expression=&self:MaterialExpressionScalarParameter_0}
    }
   ]
  }
  n3 = Custom(WaterInfoSample=@MaterialExpressionTextureSample_0.RGBA, WaterZoneIndex=@MaterialExpressionScalarParameter_0) {
   code="return DecodeWaterInfoGroundHeight(WaterInfoSample, WaterZoneIndex);"
   description=DecodeWaterInfoGroundHeight
   outputType=CMOT_Float1
   includeFilePaths [&ush]
   inputs [
    {
     inputName=WaterInfoSample
     input {
      expression=&self:MaterialExpressionTextureSample_0
      mask=1
      maskA=1
      maskB=1
      maskG=1
      maskR=1
      outputIndex=5
     }
    }
    {
     inputName=WaterZoneIndex
     input {expression=&self:MaterialExpressionScalarParameter_0}
    }
   ]
  }
  FunctionOutput(None=Subtract(
    A=FunctionOutput(None=@n2) {outputName=Water Z Height, sortPriority=1}
    B=FunctionOutput(None=@n3) {outputName=Terrain Z Height, sortPriority=2}
   )) {outputName=Water Depth, sortPriority=3}
  FunctionOutput(None=@MaterialExpressionTextureSample_0.RGBA) {outputName=Raw RGBA, sortPriority=4}
 ]}
