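# Unlit fluid simulation step material that computes ripple height (R) and foam (G) from prior render targets RT and RT2.
# A damped four-neighbor wave solve drives height and ripple foam erasure; foam averages neighbors and decays.
# Previous-frame lookups are shifted by grid offsets and water zone flow velocity.
# Forces texture adds to both channels, its alpha masks height, and a padded border mask zeroes edges.
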Material /FortWater/FluidSimulation/Materials/Simulation/M_Fluid_Sim_01
graph {
 properties {bAllowNegativeEmissiveColor=true, bUsedWithNanite=true, pixelDepthOffsetMode=PDOM_Legacy, shadingModel=MSM_Unlit, usedShadingModels=Unlit}
 nodes [
  n1 = ComponentMask(None=TextureSampleParameter2D() {parameterName=NormalAndHeight, texture=&Black}.RGB) {g=true, r=true}
  n2 = Multiply(
   A=ScalarParameter() {defaultValue=3, parameterName=Foam Advection}
   B=Divide(B=ScalarParameter() {defaultValue=1024, parameterName=Resolution})
  )
  n3 = VectorParameter() {parameterName=PreviousOffsets1}
  n4 = Subtract(A=TextureCoordinate()) {constB=0.5}
  n5 = ScalarParameter() {defaultValue=1024, parameterName=FluidSize}
  n6 = VectorParameter() {parameterName=GridLocation}
  n7 = ComponentMask(None=VectorParameter() {parameterName=WaterZoneLocation, defaultValue=<rgba -2384 6000 0 0>}.RGB) {g=true, r=true}
  n8 = ComponentMask(None=VectorParameter() {parameterName=WaterZoneExtent, defaultValue=<rgba 268800 268800 0 0>}.RGB) {g=true, r=true}
  n9 = Divide(A=@n8)
  n10 = Subtract(
   A=Add(A=Multiply(A=@n4, B=@n5), B=ComponentMask(None=@n6.RGB) {g=true, r=true})
   B=ComponentMask(None=Subtract(A=@n7, B=@n9)) {g=true, r=true}
  )
  n11 = ComponentMask(None=@n8) {g=true, r=true}
  n12 = MaterialFunctionCall!(UVs=Saturate(None=Divide(A=@n10, B=@n11))) {materialFunction=/FortWater/FluidSimulation/Materials/Functions/Fluidsim_GetWaterData}
  n13 = Divide(A=ComponentMask(None=@n12.Velocity) {g=true, r=true}, B=ScalarParameter() {defaultValue=1024, parameterName=FluidSize})
  n14 = ScalarParameter() {defaultValue=0.0167, parameterName=TS}
  n15 = StaticSwitchParameter(True=Multiply(A=Multiply(A=@n13, B=@n14), B=ScalarParameter() {defaultValue=1.6, parameterName=EnableFlow}), False=Constant()) {defaultValue=true, parameterName=Enable Water Terrain Texture}
  n16 = Subtract(A=Add(A=TextureCoordinate(), B=Multiply(A=@n1, B=@n2)), B=Add(A=ComponentMask(None=@n3.RGB) {g=true, r=true}, B=@n15))
  n17 = Divide(B=ScalarParameter() {defaultValue=1024, parameterName=Resolution}) {constA=1.5, desc="Clamp to texel center 1 texel inside "}
  Clamp(None=@n16, Min=@n17, Max=OneMinus(None=@n17))
  n18 = VectorParameter() {parameterName=PreviousOffsets2}
  n19 = Subtract(A=TextureCoordinate(), B=Add(A=ComponentMask(None=@n18.RGB) {g=true, r=true}, B=Multiply(A=@n15) {constB=2}))
  n20 = Divide(B=ScalarParameter() {defaultValue=1024, parameterName=Resolution}) {constA=1.5, desc=CellSize}
  Clamp(None=@n19, Min=@n20, Max=OneMinus(None=@n20))
  n21 = Multiply(A=ScalarParameter() {defaultValue=-0.21, parameterName=MinSimDepthToWaterDepth}, B=NamedRerouteUsage())
  n22 = VectorParameter() {parameterName=PreviousOffsets1}
  n23 = Subtract(A=TextureCoordinate(), B=Add(A=ComponentMask(None=@n22.RGB) {g=true, r=true}, B=@n15))
  n24 = TextureSampleParameter2D(UVs=@n23) {desc=Midpoint, parameterName=RT, samplerSource=SSM_Clamp_WorldGroupSettings, texture=&Black}
  n25 = Multiply(A=@n24.R) {constB=4, desc=Midpoint Height}
  n26 = MaterialFunctionCall(Float2=Divide(B=ScalarParameter() {defaultValue=1024, parameterName=Resolution}) {desc=CellSize}) {materialFunction=&BreakOutFloat2Components}
  n27 = Constant()
  n28 = AppendVector(A=@n26.R, B=@n27) {desc=X offset}
  n29 = TextureSampleParameter2D(UVs=Add(A=@n23, B=@n28)) {desc=Midpoint, parameterName=RT, samplerSource=SSM_Clamp_WorldGroupSettings, texture=&Black}
  n30 = TextureSampleParameter2D(UVs=Subtract(A=@n23, B=@n28)) {desc=N2, parameterName=RT, samplerSource=SSM_Clamp_WorldGroupSettings, texture=&Black}
  n31 = AppendVector(A=@n27, B=@n26.G) {desc=Y offset}
  n32 = TextureSampleParameter2D(UVs=Add(A=@n23, B=@n31)) {desc=N3, parameterName=RT, samplerSource=SSM_Clamp_WorldGroupSettings, texture=&Black}
  n33 = TextureSampleParameter2D(UVs=Subtract(A=@n23, B=@n31)) {desc=N4, parameterName=RT, samplerSource=SSM_Clamp_WorldGroupSettings, texture=&Black}
  n34 = DotProduct(
   A=MaterialFunctionCall(X=@n29.R, Y=@n30.R, Z=@n32.R, A=@n33.R) {desc=Surrounding Heights, materialFunction=&MakeFloat4}.Result
   B=Constant() {r=1}
  ) {desc=Neighbors Height}
  n35 = Add(A=@n25, B=Multiply(A=Subtract(A=@n34, B=@n25) {desc=N - M}, B=ScalarParameter() {defaultValue=0.9, parameterName=TravelSpeed})) {desc=Curve}
  n36 = Multiply(
   A=Subtract(
    A=Multiply(A=@n35) {constB=0.5}
    B=TextureSampleParameter2D(UVs=@n19) {desc=Midpoint, parameterName=RT2, samplerSource=SSM_Clamp_WorldGroupSettings, texture=&Black}.R
   )
   B=ScalarParameter() {defaultValue=0.98, parameterName=Damping}
  )
  n37 = Multiply(A=@n24.R) {constB=-1}
  n38 = Multiply(A=Max(A=Subtract(A=@n37) {constB=2}) {constB=0}, B=ScalarParameter() {defaultValue=0.1, parameterName=Ripple Foam Erasure})
  n39 = Abs(None=MaterialFunctionCall(X=@n29.G, Y=@n30.G, Z=@n32.G, A=@n33.G) {desc=Surrounding Heights, materialFunction=&MakeFloat4}.Result)
  n40 = Constant() {r=1}
  n41 = Divide(A=Add(A=Abs(None=@n24.G), B=DotProduct(A=@n39, B=@n40) {desc=Neighbors Height})) {constB=5}
  n42 = Multiply(A=Saturate(None=Add(A=@n38, B=@n41)), B=ScalarParameter() {defaultValue=0.998, parameterName=FoamDamping})
  n43 = TextureCoordinate()
  n44 = TextureSampleParameter2D(UVs=@n43) {desc=Forces, parameterName=Forces, samplerSource=SSM_Clamp_WorldGroupSettings, texture=&Black}
  n45 = AppendVector(A=@n44.R, B=Abs(None=@n44.G))
  n46 = Ceil(None=Subtract(A=@n44.A, B=ScalarParameter() {defaultValue=0.5, parameterName=Mask Threshold}) {constB=0.0005})
  n47 = Constant() {r=1}
  n48 = Multiply(A=Add(A=AppendVector(A=@n36, B=@n42), B=@n45), B=Saturate(None=AppendVector(A=@n46, B=@n47)))
  Clamp(None=@n48, Min=Multiply(B=@n21) {constA=-1, constB=-1}, Max=@n21) {maxDefault=100, minDefault=-100}
  n49 = ScalarParameter() {defaultValue=1024, parameterName=Resolution}
  n50 = MaterialFunctionCall(Float2=Multiply(A=Divide(B=@n49) {desc=CellSize}, B=ScalarParameter() {defaultValue=1, parameterName=ScaleFoamKernel})) {materialFunction=&BreakOutFloat2Components}
  n51 = Constant()
  n52 = AppendVector(A=@n50.R, B=@n51) {desc=X offset}
  n53 = AppendVector(A=@n51, B=@n50.G) {desc=Y offset}
  MaterialFunctionCall(
   X=TextureSampleParameter2D(UVs=Add(A=@n16, B=@n52)) {desc=Midpoint, parameterName=RT, samplerSource=SSM_Clamp_WorldGroupSettings, texture=&Black}.G
   Y=TextureSampleParameter2D(UVs=Subtract(A=@n16, B=@n52)) {desc=N2, parameterName=RT, samplerSource=SSM_Clamp_WorldGroupSettings, texture=&Black}.G
   Z=TextureSampleParameter2D(UVs=Add(A=@n16, B=@n53)) {desc=N3, parameterName=RT, samplerSource=SSM_Clamp_WorldGroupSettings, texture=&Black}.G
   A=TextureSampleParameter2D(UVs=Subtract(A=@n16, B=@n53)) {desc=N4, parameterName=RT, samplerSource=SSM_Clamp_WorldGroupSettings, texture=&Black}.G
  ) {desc=Surrounding Heights, materialFunction=&MakeFloat4}
  MaterialFunctionCall() {materialFunction=/Water/Materials/Functions/WaterZoneData}
  n54 = Constant() {r=1}
  n55 = Subtract(A=TextureCoordinate()) {constB=0.5}
  n56 = Add(
   A=Multiply(A=Abs(None=@n55)) {constB=2}
   B=Divide(B=ScalarParameter() {defaultValue=1024, parameterName=Resolution}) {constA=6, desc=Padding}
  )
  n57 = MaterialFunctionCall(Float2=Saturate(None=Floor(None=@n56))) {materialFunction=&BreakOutFloat2Components}
  n58 = Max(A=@n57.R, B=@n57.R)
  Multiply(
   A=StaticSwitchParameter(True=@n48, False=@n48) {defaultValue=true, parameterName=Enable Water Terrain Texture}
   B=Multiply(A=StaticSwitchParameter(True=@n54, False=@n54) {defaultValue=true, parameterName=Enable Water Terrain Texture}, B=OneMinus(None=@n58))
  )
  MaterialExpressionDivide_13 = Divide(
   A=NamedRerouteDeclaration(None=Max(A=@n12.Water Depth) {constB=0}) {name=WaterDepth, nodeColor=<rgba 0 0.2969 1 1>}
   B=Divide(A=ScalarParameter() {defaultValue=256, group=Flowmaps, parameterName=Target Depth for Max Velocity Flow})
  )
  n59 = Custom!(x=@MaterialExpressionDivide_13) {
   code="return 1-exp(-x);"
   description=exp
   outputType=CMOT_Float1
   inputs [{inputName=x, input {expression=&self:MaterialExpressionDivide_13}}]
  }
  Multiply(B=@n59) {desc="Depth Based Flow Mask\r\nDisabling for now since we cannot access normalized Water Texture info inside RT"}
  Multiply(A=ScalarParameter() {parameterName=ApplyForces}, B=@n45)
  n60 = NamedRerouteUsage()
  n61 = ScalarParameter() {defaultValue=0.02, parameterName=EdgeMaskOffset}
  Saturate(None=Divide(A=Subtract(A=@n60, B=@n61), B=ScalarParameter() {defaultValue=0.0005, parameterName=MaskWidth}))
  n62 = VectorParameter() {parameterName=PreviousOffsets1}
  Saturate(None=Subtract(A=@n43, B=ComponentMask(None=@n62.RGB) {g=true, r=true}))
  TextureSampleParameter2D(UVs=@n16) {desc=Midpoint, parameterName=RT, samplerSource=SSM_Clamp_WorldGroupSettings, texture=&Black}
 ]
}
